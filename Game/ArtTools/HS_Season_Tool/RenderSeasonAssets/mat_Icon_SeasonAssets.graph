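# Translucent UI material in the HS Season Tool for displaying season asset icons from the RT_Icon_SeasonAssets render target.
# It samples the target in linear color and inverts its alpha with a OneMinus node.
# The graph shows no link from that node to a material output.
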Material /Game/ArtTools/HS_Season_Tool/RenderSeasonAssets/mat_Icon_SeasonAssets
graph {
 properties {bAutomaticallySetUsageInEditor=false, blendMode=BLEND_Translucent, materialDomain=MD_UI, usedShadingModels=Default Lit}
 nodes [OneMinus(None=TextureSample() {samplerType=SAMPLERTYPE_LinearColor, texture=&/RT_Icon_SeasonAssets}.A)]
}
